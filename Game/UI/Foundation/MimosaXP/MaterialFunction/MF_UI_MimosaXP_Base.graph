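# Base UI material function for the MimosaXP icon, outputting Color, Alpha Mask and a remapped, saturated Opacity.
# An icon texture parameter packs outline, backing, gradient and alpha masks into its RGBA channels.
# Five colour parameters feed nested lerps that build the layered colour.
# Radial gradients on 9-slice UVs, one centred and two offset in opposite directions, supply the shading and colour blends.
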MaterialFunction /Game/UI/Foundation/MimosaXP/MaterialFunction/MF_UI_MimosaXP_Base
graph {nodes [
  n1 = VectorParameter() {parameterName=Colour05, defaultValue=<rgba 0 0 0 1>}
  n2 = NamedRerouteDeclaration(None=VectorParameter() {parameterName=Colour04, defaultValue=<rgba 0.0437 0.0044 0.0003 1>}.RGB) {name=Colour04, nodeColor=<rgba 0 0 0 1>}
  n3 = LinearInterpolate(Alpha=NamedRerouteUsage()) {constA=0.3, constB=0}
  n4 = NamedRerouteUsage()
  n5 = NamedRerouteUsage()
  n6 = Add(A=Multiply(A=@n4, B=NamedRerouteUsage()), B=Multiply(A=@n5, B=NamedRerouteUsage()))
  n7 = LinearInterpolate(A=@n4, Alpha=@n5)
  n8 = LinearInterpolate(
   A=LinearInterpolate(A=NamedRerouteUsage(), B=NamedRerouteDeclaration(None=@n1.RGB) {name=Colour05, nodeColor=<rgba 0 0 0 1>}, Alpha=NamedRerouteUsage())
   B=LinearInterpolate(A=Multiply(A=@n2, B=@n3), B=Saturate(None=@n6), Alpha=Multiply(A=@n7) {constB=0.125})
   Alpha=NamedRerouteUsage()
  )
  n9 = NamedRerouteUsage()
  n10 = NamedRerouteUsage()
  FunctionOutput(None=LinearInterpolate(A=LinearInterpolate(A=@n8, B=@n9, Alpha=@n10), Alpha=NamedRerouteUsage())) {outputName=Color}
  n11 = NamedRerouteUsage()
  FunctionOutput(None=@n11) {outputName=Alpha Mask}
  n12 = Constant() {r=0.4}
  n13 = Constant() {r=1}
  FunctionOutput(None=Saturate(None=MaterialFunctionCall(Input=@n11, Input Low=@n12, Input High=@n13) {materialFunction=&RemapValue_0-1}.Result)) {outputName=Opacity}
  n14 = TextureSample(Tex=TextureObjectParameter() {parameterName=IconTexture, samplerType=SAMPLERTYPE_LinearColor, texture=/Game/UI/Foundation/MimosaXP/T_UI_Icon_MimosaXP}) {samplerType=SAMPLERTYPE_LinearColor, texture=&T_UI_Icon_MimosaXP}
  NamedRerouteDeclaration(None=@n14.R) {name=OutlineMask, nodeColor=<rgba 0 0 0 1>}
  NamedRerouteDeclaration(None=@n14.G) {name=BackingMask, nodeColor=<rgba 0 0 0 1>}
  n15 = NamedRerouteDeclaration(None=VectorParameter() {parameterName=Colour03, defaultValue=<rgba 0.2346 0.0123 0 1>}.RGB) {name=Colour03, nodeColor=<rgba 0 0 0 1>}
  n16 = VectorParameter() {parameterName=Colour01, defaultValue=<rgba 0.9047 0.4852 0.0409 1>}
  n17 = MaterialFunctionCall() {materialFunction=&GetUserInterfaceUV}
  n18 = Constant() {r=0.25}
  n19 = NamedRerouteDeclaration(None=MaterialFunctionCall(UV=Add(A=@n17.9-Slice UV, B=@n18)) {materialFunction=&MF_UI_RadialGradient}.Smooth) {name=Top Radial Mask, nodeColor=<rgba 0 0 0 1>}
  n20 = VectorParameter() {parameterName=Colour02, defaultValue=<rgba 0.7157 0.0273 0.8714 1>}
  n21 = Multiply(A=@n18) {constB=-1}
  n22 = NamedRerouteDeclaration(None=MaterialFunctionCall(UV=Add(A=@n17.9-Slice UV, B=@n21)) {materialFunction=&MF_UI_RadialGradient}.Smooth) {name=Bottom Radial Mask, nodeColor=<rgba 0 0 0 1>}
  n23 = Add(
   A=LinearInterpolate(A=@n15, B=NamedRerouteDeclaration(None=@n16.RGB) {name=Colour01, nodeColor=<rgba 0 0 0 1>}, Alpha=@n19)
   B=LinearInterpolate(A=@n15, B=NamedRerouteDeclaration(None=@n20.RGB) {name=Colour02, nodeColor=<rgba 0 0 0 1>}, Alpha=@n22)
  )
  NamedRerouteDeclaration(None=LinearInterpolate(A=NamedRerouteUsage(), B=Saturate(None=@n23), Alpha=LinearInterpolate(A=@n19, Alpha=@n22))) {name=Gradient, nodeColor=<rgba 0 0 0 1>}
  NamedRerouteDeclaration(None=@n14.B) {name=GradientMask, nodeColor=<rgba 0 0 0 1>}
  NamedRerouteDeclaration(None=@n14.A) {name=Alpha, nodeColor=<rgba 0 0 0 1>}
  NamedRerouteDeclaration(None=MaterialFunctionCall(UV=MaterialFunctionCall() {materialFunction=&GetUserInterfaceUV}.9-Slice UV) {materialFunction=&MF_UI_RadialGradient}.Smooth) {name=Centered Radial Gradient, nodeColor=<rgba 0 0 0 1>}
 ]}
